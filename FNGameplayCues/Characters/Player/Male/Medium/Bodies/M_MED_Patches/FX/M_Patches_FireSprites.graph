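# Unlit translucent Niagara sprite material that draws procedural fire for the Patches character's body FX.
# A panning noise, scaled by the dynamic Distortion parameter, warps UVs for a vertically skewed soft sphere mask and scrolling wispy fire noise.
# Subtracting that noise from the mask gives the flame shape, colored through a fire curve atlas and particle color.
# Opacity uses smoothstep and particle alpha, with depth fade skipped on forward and mobile paths.
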Material /FNGameplayCues/Characters/Player/Male/Medium/Bodies/M_MED_Patches/FX/M_Patches_FireSprites
graph {
 properties {bUsedWithNiagaraSprites=true, blendMode=BLEND_Translucent, pixelDepthOffsetMode=PDOM_Legacy, shadingModel=MSM_Unlit, usedShadingModels=Unlit}
 nodes [
  n1 = Panner(Coordinate=Multiply(A=TextureCoordinate(), B=ScalarParameter() {defaultValue=1, parameterName=Distortion Scale, sliderMax=5}) {constB=2}) {speedY=0.5}
  n2 = Multiply(
   A=Subtract(A=TextureSample(UVs=@n1) {samplerType=SAMPLERTYPE_Grayscale, texture=/Game/UI/Frontend/ItemShop/Art/Rumble/LowResBlurredNoise_Rumble}.R) {constB=0.5}
   B=DynamicParameter() {defaultValue=<rgba 10 0.45 0.1 0.2>, paramNames [Gradient Bias, Radius, Distortion, Thickness]}.Distortion
  ) {constB=0}
  n3 = Add(A=TextureCoordinate(), B=@n2)
  n4 = TextureCoordinate()
  n5 = AppendVector(A=Constant() {r=1}, B=Power(Base=ComponentMask(None=@n4) {g=true}) {constExponent=1})
  n6 = SphereMask(
   A=MaterialFunctionCall(UVs=Multiply(A=@n3, B=@n5), Texture Scale=Constant2Vector() {g=1, r=0.6}) {materialFunction=&ScaleUVsByCenter}.UVs
   B=Constant2Vector() {g=0.5, r=0.5}
  ) {attenuationRadius=0.5, hardnessPercent=0}
  n7 = Power(Base=@n6, Exp=ScalarParameter() {defaultValue=2, parameterName=Ring Shape Power, sliderMax=5})
  n8 = ComponentMask(None=TextureCoordinate()) {g=true}
  n9 = DynamicParameter() {defaultValue=<rgba 10 0.45 0.1 0.2>, paramNames [Gradient Bias, Radius, Distortion, Thickness]}
  n10 = Multiply(A=Power(Base=OneMinus(None=@n8)) {constExponent=3}, B=@n9.Gradient Bias) {constB=0.5}
  ComponentMask(None=MaterialFunctionCall(Base=@n7, Blend=Add(A=@n10) {constB=0}) {materialFunction=&Blend_Overlay}.Result) {r=true}
  n11 = ParticleColor()
  n12 = Multiply(A=@n7, B=ScalarParameter() {defaultValue=3, parameterName=Ring Shape Multiply, sliderMax=5}) {constB=3}
  n13 = Add(A=TextureCoordinate(), B=@n2)
  n14 = AppendVector(
   A=ScalarParameter() {defaultValue=1, parameterName=Scale X, sliderMax=5}
   B=ScalarParameter() {defaultValue=1, parameterName=Scale Y, sliderMax=5}
  )
  n15 = ScalarParameter() {parameterName=Speed X, sliderMax=5}
  n16 = ScalarParameter() {defaultValue=0.2, parameterName=Speed Y, sliderMax=5}
  n17 = MaterialFunctionCall!(UV, Twist Center, Twist Edges) {materialFunction=/Game/Accessories/FORT_Backpacks/Backpack_F_MED_Celestial_Glow/FX/Materials/MF_PolarUV}
  n18 = TextureSample!(
   UVs=Panner(Coordinate=Multiply(A=@n13, B=@n14) {constB=2}, Speed=AppendVector(A=@n15, B=@n16)) {speedY=-0.5}
   Tex
   "DDX(UVs)"=@n17.DDX
   "DDY(UVs)"=@n17.DDY
   Apply View MipBias
  ) {mipValueMode=TMVM_Derivative, texture=/Game/Effects/Fort_Effects/Textures/T_WispyFireNoise03}
  n19 = Multiply(A=@n11.A, B=Subtract(A=@n12, B=@n18.R))
  n20 = SmoothStep(
   Min=ScalarParameter() {defaultValue=0.3, parameterName=Color SmoothStep Min, sliderMax=3, sliderMin=-3}
   Max=ScalarParameter() {defaultValue=1.2, parameterName=Color SmoothStep Max, sliderMax=3, sliderMin=-3}
   Value=Power(Base=@n19, Exp=ScalarParameter() {defaultValue=1, parameterName=Color Power, sliderMax=5})
  ) {constMax=1.2, constMin=0.3}
  n21 = CurveAtlasRowParameter(CurveTime=@n20) {atlas=/Game/Athena/LinearColorCurves/VFX_CurveAtlas_Fire, curve=/Game/Athena/LinearColorCurves/Curve_Fire03, parameterName=CurveAtlas}
  MaterialFunctionCall(Emissive=Multiply(A=@n11.RGB, B=StaticSwitchParameter(True=@n21, False=@n20) {defaultValue=true, parameterName=Use Curve Color})) {materialFunction=&MF_FX_EmissiveAdjustments}
  Multiply(A=@n6) {constB=3.5}
  Multiply(A=@n12, B=@n18.R)
  n22 = DepthFade(Opacity=@n19, FadeDistance=ScalarParameter() {defaultValue=35, parameterName=Depth Fade Distance, sliderMax=100})
  n23 = ShadingPathSwitch(Default=@n22, Deferred=@n22, Forward=@n19, Mobile=@n19)
  Multiply(
   A=SmoothStep(Min=ScalarParameter() {parameterName=Min}, Max=ScalarParameter() {defaultValue=1, parameterName=Max, sliderMax=2}, Value=@n23) {constMax=0.5}
   B=@n11.A
  )
  Multiply(A=@n6) {constB=6}
  Power(Base=@n23) {constExponent=1}
  SphereMask(A=Add(A=TextureCoordinate(), B=@n2), B=Constant() {r=0.5}, Radius=@n9.Radius) {attenuationRadius=0.3, hardnessPercent=0}
 ]
}
